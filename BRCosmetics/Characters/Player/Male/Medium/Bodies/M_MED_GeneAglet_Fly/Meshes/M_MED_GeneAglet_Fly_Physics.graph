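# Physics asset for the GeneAglet Fly male medium body cosmetic, giving its underskirt secondary motion.
# Seven three-capsule underskirt chains hang from small master bodies, alongside pelvis, thigh and calf capsules and two unconstrained sash master bodies.
# Swing limits are tightest on the A chains, whose roots lock one swing axis, and widest on C and D, with minimal twist.
# Paired, angularly free constraints cross-link the left and right A middle segments.
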PhysicsAsset /BRCosmetics/Characters/Player/Male/Medium/Bodies/M_MED_GeneAglet_Fly/Meshes/M_MED_GeneAglet_Fly_Physics
bodies {
 calf_l [{shapeType=Capsule, length=48.8835, radius=9.428, shapeName=thigh_l_capsule, center=<-23.7837 -1.5 0>, rotation=<rot 90 0 0>}]
 calf_r [{shapeType=Capsule, length=49.3288, radius=9.4374, shapeName=thigh_r_capsule, center=<24.055 1.5 0>, rotation=<rot 90 0 0>}]
 dyn_sash_l_MSTR [{shapeType=Capsule, length=0.505, radius=0.505, shapeName=dyn_sash_l_MSTR_capsule, rotation=<rot 0 -89.9999 10.1939>}]
 dyn_sash_r_MSTR [{shapeType=Capsule, length=0.505, radius=0.505, shapeName=dyn_sash_r_MSTR_capsule, rotation=<rot 0 -90 0>}]
 dyn_underskirt_A_l_01 [{
   shapeType=Capsule
   length=31.4515
   radius=8.9922
   shapeName=dyn_underskirt_A_l_01_capsule
   center=<9.8889 0.7177 12>
   rotation=<rot 0.0005 0.0002 89.0535>
  }]
 dyn_underskirt_A_l_02 [{
   shapeType=Capsule
   length=30.1495
   radius=12.4291
   shapeName=dyn_underskirt_A_l_02_capsule
   center=<5.5533 0.1348 11.4>
   rotation=<rot 0.0004 0.0001 89.0535>
  }]
 dyn_underskirt_A_l_03 [{
   shapeType=Capsule
   length=26.5603
   radius=12.4291
   shapeName=dyn_underskirt_A_l_03_capsule
   center=<5.4328 0.6331 8.9>
   rotation=<rot 0.0004 0.0001 89.0535>
  }]
 dyn_underskirt_A_l_MSTR [{shapeType=Capsule, length=0.505, radius=0.505, shapeName=dyn_underskirt_A_l_MSTR_capsule, rotation=<rot 0 -89.9999 10.1939>}]
 dyn_underskirt_A_r_01 [{
   shapeType=Capsule
   length=29.9972
   radius=8.9922
   shapeName=dyn_underskirt_A_r_01_capsule
   center=<-9.937 -0.8967 -12>
   rotation=<rot 0.0008 0.0001 83.5502>
  }]
 dyn_underskirt_A_r_02 [{
   shapeType=Capsule
   length=28.8019
   radius=12.4619
   shapeName=dyn_underskirt_A_r_02_capsule
   center=<-5.5532 0.1571 -11.4>
   rotation=<rot 0.0005 0.0001 78.6939>
  }]
 dyn_underskirt_A_r_03 [{
   shapeType=Capsule
   length=26.5275
   radius=12.4619
   shapeName=dyn_underskirt_A_r_03_capsule
   center=<-5.2587 -0.1507 -9.2>
   rotation=<rot 0.0005 0.0001 78.6939>
  }]
 dyn_underskirt_A_r_MSTR [{shapeType=Capsule, length=0.505, radius=0.505, shapeName=dyn_underskirt_A_r_MSTR_capsule, rotation=<rot 0 -90 0>}]
 dyn_underskirt_B_l_01 [{
   shapeType=Capsule
   length=35.0969
   radius=10.2315
   shapeName=dyn_underskirt_B_l_01_capsule
   center=<11.1298 1.3926 10.6>
   rotation=<rot 0 0.0002 107.1567>
  }]
 dyn_underskirt_B_l_02 [{
   shapeType=Capsule
   length=33.8775
   radius=11.4509
   shapeName=dyn_underskirt_B_l_02_capsule
   center=<7.4144 1.8158 10.3>
   rotation=<rot 0 0.0001 109.4328>
  }]
 dyn_underskirt_B_l_03 [{
   shapeType=Capsule
   length=32.9104
   radius=12.4179
   shapeName=dyn_underskirt_B_l_03_capsule
   center=<6.4664 1.9684 10.5>
   rotation=<rot 0 0.0002 108.7444>
  }]
 dyn_underskirt_B_l_MSTR [{shapeType=Capsule, length=0.505, radius=0.505, shapeName=dyn_underskirt_B_l_MSTR_capsule, rotation=<rot 0 -90 0>}]
 dyn_underskirt_B_r_01 [{
   shapeType=Capsule
   length=33.8211
   radius=11.5283
   shapeName=dyn_underskirt_B_r_01_capsule
   center=<-10.8056 -0.8274 -12.4>
   rotation=<rot 0 0.0004 100.3684>
  }]
 dyn_underskirt_B_r_02 [{
   shapeType=Capsule
   length=33.1776
   radius=12.9646
   shapeName=dyn_underskirt_B_r_02_capsule
   center=<-5.0409 -1.2512 -11.7>
   rotation=<rot 0 0.0003 103.4894>
  }]
 dyn_underskirt_B_r_03 [{
   shapeType=Capsule
   length=32.2016
   radius=13.1478
   shapeName=dyn_underskirt_B_r_03_capsule
   center=<-5.668 -0.5767 -9.8>
   rotation=<rot 0 0.0004 102.4189>
  }]
 dyn_underskirt_B_r_MSTR [{shapeType=Capsule, length=0.505, radius=0.505, shapeName=dyn_underskirt_B_r_MSTR_capsule, rotation=<rot 0 -90 0>}]
 dyn_underskirt_C_l_01 [{
   shapeType=Capsule
   length=46.3041
   radius=10.695
   shapeName=dyn_underskirt_C_l_01_capsule
   center=<11.3018 1.2931 10.8>
   rotation=<rot -0.0002 0.0001 112.7125>
  }]
 dyn_underskirt_C_l_02 [{
   shapeType=Capsule
   length=45.7082
   radius=11.2909
   shapeName=dyn_underskirt_C_l_02_capsule
   center=<5.5149 1.2931 10.8>
   rotation=<rot -0.0002 0.0001 112.7125>
  }]
 dyn_underskirt_C_l_03 [{
   shapeType=Capsule
   length=45.0226
   radius=11.9765
   shapeName=dyn_underskirt_C_l_03_capsule
   center=<3.9035 1.2931 10.8>
   rotation=<rot -0.0002 0.0001 112.7125>
  }]
 dyn_underskirt_C_l_MSTR [{shapeType=Capsule, length=0.505, radius=0.505, shapeName=dyn_underskirt_C_l_MSTR_capsule, rotation=<rot 0 -90 0>}]
 dyn_underskirt_C_r_01 [{
   shapeType=Capsule
   length=51.5707
   radius=10.4234
   shapeName=dyn_underskirt_C_r_01_capsule
   center=<-10.2024 -0.925 -10.8>
   rotation=<rot -0.0001 0 101.3121>
  }]
 dyn_underskirt_C_r_02 [{
   shapeType=Capsule
   length=50.5311
   radius=11.4631
   shapeName=dyn_underskirt_C_r_02_capsule
   center=<-4.6252 -0.898 -10.8>
   rotation=<rot -0.0001 0 101.3121>
  }]
 dyn_underskirt_C_r_03 [{
   shapeType=Capsule
   length=49.6661
   radius=12.3281
   shapeName=dyn_underskirt_C_r_03_capsule
   center=<-3.2917 -1.0606 -10.8>
   rotation=<rot -0.0001 0 101.3121>
  }]
 dyn_underskirt_C_r_MSTR [{shapeType=Capsule, length=0.505, radius=0.505, shapeName=dyn_underskirt_C_r_MSTR_capsule, rotation=<rot 0 -90 0>}]
 dyn_underskirt_D_01 [{shapeType=Capsule, length=29.0789, radius=10.4545, shapeName=dyn_underskirt_D_01_capsule, center=<9.823 0 8>, rotation=<rot 0 0 90>}]
 dyn_underskirt_D_02 [{shapeType=Capsule, length=28.5452, radius=10.9882, shapeName=dyn_underskirt_D_02_capsule, center=<5.2766 0 8>, rotation=<rot 0 0 90>}]
 dyn_underskirt_D_03 [{shapeType=Capsule, length=27.7749, radius=11.7586, shapeName=dyn_underskirt_D_03_capsule, center=<5.9283 0 8>, rotation=<rot 0 0 90>}]
 dyn_underskirt_D_MSTR [{shapeType=Capsule, length=0.505, radius=0.505, shapeName=dyn_underskirt_D_MSTR_capsule, rotation=<rot 0 -90 0>}]
 pelvis [{shapeType=Capsule, length=13.8373, radius=9.955, shapeName=pelvis_capsule, rotation=<rot 0 -0.0002 10.1941>}]
 thigh_l [{
   shapeType=Capsule
   length=43.7098
   radius=9.7004
   shapeName=thigh_l_capsule
   center=<-22.0844 -0.5412 1.4148>
   rotation=<rot 90 180 177.9347>
  }]
 thigh_r [{
   shapeType=Capsule
   length=43.1991
   radius=10.731
   shapeName=thigh_r_capsule
   center=<22.1084 0.13 -1.9212>
   rotation=<rot 88.2841 3.53 1.1728>
  }]
}
constraints [
 {bone1Name=dyn_underskirt_A_l_02, bone2Name=dyn_underskirt_A_l_01, swing1LimitDegrees=8, swing2LimitDegrees=6, twistLimitDegrees=2}
 {
  bone1Name=dyn_underskirt_A_l_01
  bone2Name=dyn_underskirt_A_l_MSTR
  swing1LimitDegrees=1
  swing1Motion=Locked
  swing2LimitDegrees=4
  twistLimitDegrees=2
 }
 {bone1Name=dyn_underskirt_A_l_03, bone2Name=dyn_underskirt_A_l_02, swing1LimitDegrees=8, swing2LimitDegrees=6, twistLimitDegrees=2}
 {
  bone1Name=dyn_underskirt_A_r_01
  bone2Name=dyn_underskirt_A_r_MSTR
  swing1LimitDegrees=1
  swing1Motion=Locked
  swing2LimitDegrees=4
  twistLimitDegrees=2
 }
 {bone1Name=dyn_underskirt_A_r_02, bone2Name=dyn_underskirt_A_r_01, swing1LimitDegrees=8, swing2LimitDegrees=6, twistLimitDegrees=2}
 {bone1Name=dyn_underskirt_A_r_03, bone2Name=dyn_underskirt_A_r_02, swing1LimitDegrees=8, swing2LimitDegrees=6, twistLimitDegrees=2}
 {bone1Name=dyn_underskirt_B_l_01, bone2Name=dyn_underskirt_B_l_MSTR, swing1LimitDegrees=4, swing2LimitDegrees=4, twistLimitDegrees=2}
 {bone1Name=dyn_underskirt_B_l_02, bone2Name=dyn_underskirt_B_l_01, swing1LimitDegrees=10, swing2LimitDegrees=6, twistLimitDegrees=2}
 {bone1Name=dyn_underskirt_B_l_03, bone2Name=dyn_underskirt_B_l_02, swing1LimitDegrees=10, swing2LimitDegrees=6, twistLimitDegrees=2}
 {bone1Name=dyn_underskirt_B_r_01, bone2Name=dyn_underskirt_B_r_MSTR, swing1LimitDegrees=4, swing2LimitDegrees=4, twistLimitDegrees=2}
 {bone1Name=dyn_underskirt_B_r_02, bone2Name=dyn_underskirt_B_r_01, swing1LimitDegrees=10, swing2LimitDegrees=6, twistLimitDegrees=2}
 {bone1Name=dyn_underskirt_B_r_03, bone2Name=dyn_underskirt_B_r_02, swing1LimitDegrees=10, swing2LimitDegrees=6, twistLimitDegrees=2}
 {bone1Name=dyn_underskirt_C_l_01, bone2Name=dyn_underskirt_C_l_MSTR, swing1LimitDegrees=8, swing2LimitDegrees=16, twistLimitDegrees=2}
 {bone1Name=dyn_underskirt_C_l_02, bone2Name=dyn_underskirt_C_l_01, swing1LimitDegrees=10, swing2LimitDegrees=20, twistLimitDegrees=2}
 {bone1Name=dyn_underskirt_C_l_03, bone2Name=dyn_underskirt_C_l_02, swing1LimitDegrees=16, swing2LimitDegrees=20, twistLimitDegrees=2}
 {bone1Name=dyn_underskirt_C_r_01, bone2Name=dyn_underskirt_C_r_MSTR, swing1LimitDegrees=8, swing2LimitDegrees=16, twistLimitDegrees=2}
 {bone1Name=dyn_underskirt_C_r_02, bone2Name=dyn_underskirt_C_r_01, swing1LimitDegrees=10, swing2LimitDegrees=20, twistLimitDegrees=2}
 {bone1Name=dyn_underskirt_C_r_03, bone2Name=dyn_underskirt_C_r_02, swing1LimitDegrees=16, swing2LimitDegrees=20, twistLimitDegrees=2}
 {bone1Name=dyn_underskirt_D_01, bone2Name=dyn_underskirt_D_MSTR, swing1LimitDegrees=8, swing2LimitDegrees=16, twistLimitDegrees=2}
 {bone1Name=dyn_underskirt_D_02, bone2Name=dyn_underskirt_D_01, swing1LimitDegrees=10, swing2LimitDegrees=20, twistLimitDegrees=2}
 {bone1Name=dyn_underskirt_D_03, bone2Name=dyn_underskirt_D_02, swing1LimitDegrees=16, swing2LimitDegrees=20, twistLimitDegrees=2}
 {
  bone1Name=dyn_underskirt_A_l_02
  bone2Name=dyn_underskirt_A_r_02
  swing1LimitDegrees=45
  swing1Motion=Free
  swing2LimitDegrees=45
  swing2Motion=Free
  twistLimitDegrees=45
  twistMotion=Free
 }
 {
  bone1Name=dyn_underskirt_A_r_02
  bone2Name=dyn_underskirt_A_l_02
  swing1LimitDegrees=45
  swing1Motion=Free
  swing2LimitDegrees=45
  swing2Motion=Free
  twistLimitDegrees=45
  twistMotion=Free
 }
]
