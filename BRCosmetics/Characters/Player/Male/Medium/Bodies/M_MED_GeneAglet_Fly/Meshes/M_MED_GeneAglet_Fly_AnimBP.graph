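# Animation Blueprint for the GeneAglet_Fly player body cosmetic, adding secondary motion to sleeves, underskirt, sash and lightsaber bones.
# It copies the parent mesh's pose through a Control Rig fed knee-sliding and front-end flags, then three RigidBody simulations for body and sleeves.
# AnimDynamics, additive bone offsets and constraints pinning sash to underskirt bones follow.
# Other sleeve, skirt and lightsaber chains lack pose inputs, apparently unused.
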AnimBlueprint /BRCosmetics/Characters/Player/Male/Medium/Bodies/M_MED_GeneAglet_Fly/Meshes/M_MED_GeneAglet_Fly_AnimBP
target_skeleton=&/M_MED_GeneAglet_Fly_Skeleton
graphs [
 AnimGraph {nodes [
   n1 = Anim Dynamics - Bone: dyn_sleeve_top_l!(ComponentPose, Alpha=1, GravityOverride:Vector="0,0,-980", ExternalForce="0,0,0") -> (Pose) {
    angularDampingOverride=0.9999
    angularSpringConstant=50
    bAngularSpring=true
    bLinearSpring=true
    bOverrideAngularDamping=true
    bOverrideLinearDamping=true
    bUseGravityOverride=true
    linearDampingOverride=0.9999
    linearSpringConstant=50
    simulationSpace=RootRelative
    alphaBoolBlend {blendOption=Linear}
    boundBone {boneName=dyn_sleeve_top_l}
    componentAppliedLinearAccClamp=<3000 3000 3000>
    componentLinearAccScale=<0.1 0.1 0.1>
    componentLinearVelScale=<0.2 0.2 0.2>
    gravityOverride {z=-980}
    physicsBodyDefinitions [{
      boundBone {boneName=dyn_sleeve_top_l}
      boxExtents=<40 40 40>
      constraintSetup {angularLimitsMax=<4 0 2>, angularLimitsMin=<-4 -4 0>, angularTarget=<1 0 0>, linearAxesMin=<-2 0 0>}
      localJointOffset=<0.2 0 0>
     }]
    relativeSpaceBone {boneName=lowerarm_l}
   }
   n2 = Anim Dynamics - Bone: dyn_sleeve_bottom_l!(ComponentPose=@n1, Alpha=1, GravityOverride:Vector="0,0,-980", ExternalForce="0,0,0") -> (Pose) {
    angularDampingOverride=0.9999
    angularSpringConstant=50
    bAngularSpring=true
    bLinearSpring=true
    bOverrideAngularDamping=true
    bOverrideLinearDamping=true
    bUseGravityOverride=true
    linearDampingOverride=0.9999
    linearSpringConstant=50
    simulationSpace=RootRelative
    alphaBoolBlend {blendOption=Linear}
    boundBone {boneName=dyn_sleeve_bottom_l}
    componentAppliedLinearAccClamp=<3000 3000 3000>
    componentLinearAccScale=<0.1 0.1 0.1>
    componentLinearVelScale=<0.2 0.2 0.2>
    gravityOverride {z=-980}
    physicsBodyDefinitions [{
      boundBone {boneName=dyn_sleeve_bottom_l}
      boxExtents=<40 40 40>
      constraintSetup {angularLimitsMax=<4 10 4>, angularLimitsMin=<-4 -10 -4>, angularTarget=<1 0 0>, linearAxesMax=<0 2 2>, linearAxesMin=<-2 -2 -2>}
      localJointOffset=<0.2 0 0>
     }]
    relativeSpaceBone {boneName=lowerarm_l}
   }
   n3 = Anim Dynamics - Bone: dyn_sleeve_top_r!(ComponentPose=@n2, Alpha=1, GravityOverride:Vector="0,0,-980", ExternalForce="0,0,0") -> (Pose) {
    angularDampingOverride=0.9999
    angularSpringConstant=50
    bAngularSpring=true
    bLinearSpring=true
    bOverrideAngularDamping=true
    bOverrideLinearDamping=true
    bUseGravityOverride=true
    linearDampingOverride=0.9999
    linearSpringConstant=50
    simulationSpace=RootRelative
    alphaBoolBlend {blendOption=Linear}
    boundBone {boneName=dyn_sleeve_top_r}
    componentAppliedLinearAccClamp=<3000 3000 3000>
    componentLinearAccScale=<0.1 0.1 0.1>
    componentLinearVelScale=<0.2 0.2 0.2>
    gravityOverride {z=-980}
    physicsBodyDefinitions [{
      boundBone {boneName=dyn_sleeve_top_r}
      boxExtents=<40 40 40>
      constraintSetup {angularLimitsMax=<4 0 2>, angularLimitsMin=<-4 -4 0>, angularTarget=<1 0 0>, linearAxesMin=<-2 0 0>}
      localJointOffset=<0.2 0 0>
     }]
    relativeSpaceBone {boneName=lowerarm_r}
   }
   Anim Dynamics - Bone: dyn_sleeve_bottom_r!(ComponentPose=@n3, Alpha=1, GravityOverride:Vector="0,0,-980", ExternalForce="0,0,0") -> (Pose) {
    angularDampingOverride=0.9999
    angularSpringConstant=50
    bAngularSpring=true
    bLinearSpring=true
    bOverrideAngularDamping=true
    bOverrideLinearDamping=true
    bUseGravityOverride=true
    linearDampingOverride=0.9999
    linearSpringConstant=50
    simulationSpace=RootRelative
    alphaBoolBlend {blendOption=Linear}
    boundBone {boneName=dyn_sleeve_bottom_r}
    componentAppliedLinearAccClamp=<3000 3000 3000>
    componentLinearAccScale=<0.1 0.1 0.1>
    componentLinearVelScale=<0.2 0.2 0.2>
    gravityOverride {z=-980}
    physicsBodyDefinitions [{
      boundBone {boneName=dyn_sleeve_bottom_r}
      boxExtents=<40 40 40>
      constraintSetup {angularLimitsMax=<4 10 4>, angularLimitsMin=<-4 -10 -4>, angularTarget=<1 0 0>, linearAxesMax=<0 2 2>, linearAxesMin=<-2 -2 -2>}
      localJointOffset=<0.2 0 0>
     }]
    relativeSpaceBone {boneName=lowerarm_r}
   }
   n4 = Anim Dynamics - Bone: dyn_underskirt_A_l_01!(ComponentPose, Alpha=1, GravityOverride:Vector="0,0,-980", ExternalForce="0,0,0") -> (Pose) {
    angularDampingOverride=0.9999
    angularSpringConstant=50
    bOverrideAngularDamping=true
    bUseGravityOverride=true
    simulationSpace=RootRelative
    alphaBoolBlend {blendOption=Linear}
    boundBone {boneName=dyn_underskirt_A_l_01}
    componentLinearAccScale=<0.4 0.4 0.4>
    componentLinearVelScale=<0.5 0.5 0.5>
    gravityOverride {z=-980}
    physicsBodyDefinitions [{
      boundBone {boneName=dyn_underskirt_A_l_01}
      boxExtents=<10 10 10>
      constraintSetup {angularLimitsMax=<0 0 0>, angularLimitsMin=<0 -6 0>, angularTarget=<1 0 0>}
      localJointOffset=<1 0 0>
     }]
   }
   n5 = Anim Dynamics - Bone: dyn_underskirt_A_l_02!(ComponentPose=@n4, Alpha=1, GravityOverride:Vector="0,0,-980", ExternalForce="0,0,0") -> (Pose) {
    angularDampingOverride=0.9999
    angularSpringConstant=50
    bOverrideAngularDamping=true
    bUseGravityOverride=true
    simulationSpace=RootRelative
    alphaBoolBlend {blendOption=Linear}
    boundBone {boneName=dyn_underskirt_A_l_02}
    componentLinearAccScale=<0.4 0.4 0.4>
    componentLinearVelScale=<0.5 0.5 0.5>
    gravityOverride {z=-980}
    physicsBodyDefinitions [{
      boundBone {boneName=dyn_underskirt_A_l_02}
      boxExtents=<10 10 10>
      constraintSetup {angularLimitsMax=<0 0 0>, angularLimitsMin=<0 -6 0>, angularTarget=<1 0 0>}
      localJointOffset=<1 0 0>
     }]
   }
   n6 = Anim Dynamics - Bone: dyn_underskirt_A_l_03!(ComponentPose=@n5, Alpha=1, GravityOverride:Vector="0,0,-980", ExternalForce="0,0,0") -> (Pose) {
    angularDampingOverride=0.9999
    angularSpringConstant=50
    bOverrideAngularDamping=true
    bUseGravityOverride=true
    simulationSpace=RootRelative
    alphaBoolBlend {blendOption=Linear}
    boundBone {boneName=dyn_underskirt_A_l_03}
    componentLinearAccScale=<0.4 0.4 0.4>
    componentLinearVelScale=<0.5 0.5 0.5>
    gravityOverride {z=-980}
    physicsBodyDefinitions [{
      boundBone {boneName=dyn_underskirt_A_l_03}
      boxExtents=<10 10 10>
      constraintSetup {angularLimitsMax=<0 0 0>, angularLimitsMin=<0 -6 0>, angularTarget=<1 0 0>}
      localJointOffset=<1 0 0>
     }]
   }
   n7 = Anim Dynamics - Bone: dyn_underskirt_A_r_01!(ComponentPose=@n6, Alpha=1, GravityOverride:Vector="0,0,-980", ExternalForce="0,0,0") -> (Pose) {
    angularDampingOverride=0.9999
    angularSpringConstant=50
    bOverrideAngularDamping=true
    bUseGravityOverride=true
    simulationSpace=RootRelative
    alphaBoolBlend {blendOption=Linear}
    boundBone {boneName=dyn_underskirt_A_r_01}
    componentLinearAccScale=<0.4 0.4 0.4>
    componentLinearVelScale=<0.5 0.5 0.5>
    gravityOverride {z=-980}
    physicsBodyDefinitions [{
      boundBone {boneName=dyn_underskirt_A_r_01}
      boxExtents=<10 10 10>
      constraintSetup {angularLimitsMax=<0 0 0>, angularLimitsMin=<0 -6 0>, angularTarget=<1 0 0>}
      localJointOffset=<1 0 0>
     }]
   }
   n8 = Anim Dynamics - Bone: dyn_underskirt_A_r_02!(ComponentPose=@n7, Alpha=1, GravityOverride:Vector="0,0,-980", ExternalForce="0,0,0") -> (Pose) {
    angularDampingOverride=0.9999
    angularSpringConstant=50
    bOverrideAngularDamping=true
    bUseGravityOverride=true
    simulationSpace=RootRelative
    alphaBoolBlend {blendOption=Linear}
    boundBone {boneName=dyn_underskirt_A_r_02}
    componentLinearAccScale=<0.4 0.4 0.4>
    componentLinearVelScale=<0.5 0.5 0.5>
    gravityOverride {z=-980}
    physicsBodyDefinitions [{
      boundBone {boneName=dyn_underskirt_A_r_02}
      boxExtents=<10 10 10>
      constraintSetup {angularLimitsMax=<0 0 0>, angularLimitsMin=<0 -6 0>, angularTarget=<1 0 0>}
      localJointOffset=<1 0 0>
     }]
   }
   Anim Dynamics - Bone: dyn_underskirt_A_r_03!(ComponentPose=@n8, Alpha=1, GravityOverride:Vector="0,0,-980", ExternalForce="0,0,0") -> (Pose) {
    angularDampingOverride=0.9999
    angularSpringConstant=50
    bOverrideAngularDamping=true
    bUseGravityOverride=true
    simulationSpace=RootRelative
    alphaBoolBlend {blendOption=Linear}
    boundBone {boneName=dyn_underskirt_A_r_03}
    componentLinearAccScale=<0.4 0.4 0.4>
    componentLinearVelScale=<0.5 0.5 0.5>
    gravityOverride {z=-980}
    physicsBodyDefinitions [{
      boundBone {boneName=dyn_underskirt_A_r_03}
      boxExtents=<10 10 10>
      constraintSetup {angularLimitsMax=<0 0 0>, angularLimitsMin=<0 -6 0>, angularTarget=<1 0 0>}
      localJointOffset=<1 0 0>
     }]
   }
   Anim Dynamics - Bone: dyn_lightsaber!(ComponentPose, Alpha=1, GravityOverride:Vector="0,0,-980", ExternalForce="0,0,0") -> (Pose) {
    angularDampingOverride=0.9999
    angularSpringConstant=100
    bAngularSpring=true
    bOverrideAngularDamping=true
    bUseGravityOverride=true
    simulationSpace=RootRelative
    alphaBoolBlend {blendOption=Linear}
    boundBone {boneName=dyn_lightsaber}
    componentLinearAccScale=<0.4 0.4 0.4>
    componentLinearVelScale=<0.5 0.5 0.5>
    gravityOverride {z=-980}
    physicsBodyDefinitions [{
      boundBone {boneName=dyn_lightsaber}
      boxExtents=<16 16 16>
      constraintSetup {angularLimitsMax=<2 2 10>, angularLimitsMin=<-2 -2 -10>, angularTarget=<1 0 0>}
      localJointOffset=<2.5 0 0>
     }]
   }
   Anim Dynamics - Bone: dyn_lightsaber(Alpha=1, GravityOverride="0,0,-980", ExternalForce="0,0,0") {
    angularDampingOverride=0.9999
    angularSpringConstant=100
    bAngularSpring=true
    bOverrideAngularDamping=true
    bUseGravityOverride=true
    simulationSpace=RootRelative
    alphaBoolBlend {blendOption=Linear}
    boundBone {boneName=dyn_lightsaber}
    componentLinearAccScale=<0.4 0.4 0.4>
    componentLinearVelScale=<0.5 0.5 0.5>
    gravityOverride {z=-980}
    physicsBodyDefinitions [{
      boundBone {boneName=dyn_lightsaber}
      boxExtents=<16 16 16>
      constraintSetup {angularLimitsMax=<2 10 6>, angularLimitsMin=<-2 -4 -6>, angularTarget=<1 0 0>}
      localJointOffset=<2.5 0 0>
     }]
   }
   n9 = "Transform (Modify) Bone - Bone: dyn_lowerarm_sleeve_l_MSTR"(Alpha=1, Translation="-2,0,0", Rotation="8,3.734,0.152", Scale="1,1,1") {
    rotationMode=BMM_Additive
    rotationSpace=BCS_BoneSpace
    translationMode=BMM_Additive
    translationSpace=BCS_BoneSpace
    alphaBoolBlend {blendOption=Linear}
    boneToModify {boneName=dyn_lowerarm_sleeve_l_MSTR}
    rotation=<rot 8 3.734 0.152>
    translation {x=-2}
   }
   n10 = "Transform (Modify) Bone - Bone: dyn_lowerarm_sleeve_r_MSTR"(ComponentPose=@n9, Alpha=1, Translation="2,0,0", Rotation="4,-2,0", Scale="1,1,1") {
    rotationMode=BMM_Additive
    rotationSpace=BCS_BoneSpace
    translationMode=BMM_Additive
    translationSpace=BCS_BoneSpace
    alphaBoolBlend {blendOption=Linear}
    boneToModify {boneName=dyn_lowerarm_sleeve_r_MSTR}
    rotation {pitch=4, yaw=-2}
    translation {x=2}
   }
   n11 = Anim Dynamics - Bone: dyn_lowerarm_l_sleeve_twist_03!(SimulationSpace:AnimPhysSimSpaceType Enum, ComponentPose=@n10, Alpha=0.8, ExternalForce="0,0,0") -> (Pose) {
    angularDampingOverride=1
    angularSpringConstant=1000
    bAngularSpring=true
    bLinearSpring=true
    bOverrideLinearDamping=true
    bUseGravityOverride=true
    bUsePlanarLimit=false
    gravityScale=0.35
    linearDampingOverride=1
    linearSpringConstant=50
    simulationSpace=RootRelative
    alphaBoolBlend {blendOption=Linear}
    boundBone {boneName=dyn_lowerarm_l_sleeve_twist_03}
    componentAppliedLinearAccClamp=<5 5 5>
    componentLinearAccScale=<0.05 0.05 0.05>
    componentLinearVelScale=<0.05 0.05 0.05>
    gravityOverride {z=-980}
    physicsBodyDefinitions [{
      boundBone {boneName=dyn_lowerarm_l_sleeve_twist_03}
      boxExtents=<20 20 20>
      constraintSetup {angularLimitsMax=<1 2 2>, angularLimitsMin=<-1 -2 -2>, angularTarget=<1 0 0>}
      localJointOffset=<1 0 0>
     }]
    relativeSpaceBone {boneName=root}
   }
   n12 = Anim Dynamics - Bone: dyn_lowerarm_l_sleeve_twist_02!(SimulationSpace:AnimPhysSimSpaceType Enum, ComponentPose=@n11, Alpha=0.8, ExternalForce="0,0,0") -> (Pose) {
    angularDampingOverride=1
    angularSpringConstant=1000
    bAngularSpring=true
    bLinearSpring=true
    bOverrideLinearDamping=true
    bUseGravityOverride=true
    bUsePlanarLimit=false
    gravityScale=0.35
    linearDampingOverride=1
    linearSpringConstant=50
    simulationSpace=RootRelative
    alphaBoolBlend {blendOption=Linear}
    boundBone {boneName=dyn_lowerarm_l_sleeve_twist_02}
    componentAppliedLinearAccClamp=<5 5 5>
    componentLinearAccScale=<0.05 0.05 0.05>
    componentLinearVelScale=<0.05 0.05 0.05>
    gravityOverride {z=-980}
    physicsBodyDefinitions [{
      boundBone {boneName=dyn_lowerarm_l_sleeve_twist_02}
      boxExtents=<20 20 20>
      constraintSetup {angularLimitsMax=<1 2 2>, angularLimitsMin=<-1 -2 -2>, angularTarget=<1 0 0>}
      localJointOffset=<1 0 0>
     }]
    relativeSpaceBone {boneName=root}
   }
   n13 = Anim Dynamics - Bone: dyn_lowerarm_l_sleeve_twist_03_end!(SimulationSpace:AnimPhysSimSpaceType Enum, ComponentPose=@n12, Alpha=0.8, ExternalForce="0,0,0") -> (Pose) {
    angularDampingOverride=1
    angularSpringConstant=1000
    bAngularSpring=true
    bLinearSpring=true
    bOverrideLinearDamping=true
    bUseGravityOverride=true
    bUsePlanarLimit=false
    gravityScale=0.35
    linearDampingOverride=1
    linearSpringConstant=100
    simulationSpace=RootRelative
    alphaBoolBlend {blendOption=Linear}
    boundBone {boneName=dyn_lowerarm_l_sleeve_twist_03_end}
    componentAppliedLinearAccClamp=<5 5 5>
    componentLinearAccScale=<0.05 0.05 0.05>
    componentLinearVelScale=<0.05 0.05 0.05>
    gravityOverride {z=-980}
    physicsBodyDefinitions [{
      boundBone {boneName=dyn_lowerarm_l_sleeve_twist_03_end}
      boxExtents=<15 15 15>
      constraintSetup {angularLimitsMax=<10 10 10>, angularLimitsMin=<-10 -10 -10>, angularTarget=<1 0 0>}
      localJointOffset=<2 0 0>
     }]
    relativeSpaceBone {boneName=root}
   }
   n14 = Anim Dynamics - Bone: dyn_lowerarm_l_sleeve_twist_02_end!(SimulationSpace:AnimPhysSimSpaceType Enum, ComponentPose=@n13, Alpha=0.8, ExternalForce="0,0,0") -> (Pose) {
    angularDampingOverride=1
    angularSpringConstant=1000
    bAngularSpring=true
    bLinearSpring=true
    bOverrideLinearDamping=true
    bUseGravityOverride=true
    bUsePlanarLimit=false
    gravityScale=0.35
    linearDampingOverride=1
    linearSpringConstant=100
    simulationSpace=RootRelative
    alphaBoolBlend {blendOption=Linear}
    boundBone {boneName=dyn_lowerarm_l_sleeve_twist_02_end}
    componentAppliedLinearAccClamp=<5 5 5>
    componentLinearAccScale=<0.05 0.05 0.05>
    componentLinearVelScale=<0.05 0.05 0.05>
    gravityOverride {z=-980}
    physicsBodyDefinitions [{
      boundBone {boneName=dyn_lowerarm_l_sleeve_twist_02_end}
      boxExtents=<15 15 15>
      constraintSetup {angularLimitsMax=<10 10 10>, angularLimitsMin=<-10 -10 -10>, angularTarget=<1 0 0>}
      localJointOffset=<2 0 0>
     }]
    relativeSpaceBone {boneName=root}
   }
   n15 = Anim Dynamics - Bone: dyn_lowerarm_r_sleeve_twist_03!(SimulationSpace:AnimPhysSimSpaceType Enum, ComponentPose=@n14, Alpha=0.8, ExternalForce="0,0,0") -> (Pose) {
    angularDampingOverride=1
    angularSpringConstant=1000
    bAngularSpring=true
    bLinearSpring=true
    bOverrideLinearDamping=true
    bUseGravityOverride=true
    bUsePlanarLimit=false
    gravityScale=0.35
    linearDampingOverride=1
    linearSpringConstant=50
    simulationSpace=RootRelative
    alphaBoolBlend {blendOption=Linear}
    boundBone {boneName=dyn_lowerarm_r_sleeve_twist_03}
    componentAppliedLinearAccClamp=<5 5 5>
    componentLinearAccScale=<0.05 0.05 0.05>
    componentLinearVelScale=<0.05 0.05 0.05>
    gravityOverride {z=-980}
    physicsBodyDefinitions [{
      boundBone {boneName=dyn_lowerarm_r_sleeve_twist_03}
      boxExtents=<20 20 20>
      constraintSetup {angularLimitsMax=<1 2 2>, angularLimitsMin=<-1 -2 -2>, angularTarget=<1 0 0>}
      localJointOffset=<-1 0 0>
     }]
    relativeSpaceBone {boneName=root}
   }
   n16 = Anim Dynamics - Bone: dyn_lowerarm_r_sleeve_twist_02!(SimulationSpace:AnimPhysSimSpaceType Enum, ComponentPose=@n15, Alpha=0.8, ExternalForce="0,0,0") -> (Pose) {
    angularDampingOverride=1
    angularSpringConstant=1000
    bAngularSpring=true
    bLinearSpring=true
    bOverrideLinearDamping=true
    bUseGravityOverride=true
    bUsePlanarLimit=false
    gravityScale=0.35
    linearDampingOverride=1
    linearSpringConstant=50
    simulationSpace=RootRelative
    alphaBoolBlend {blendOption=Linear}
    boundBone {boneName=dyn_lowerarm_r_sleeve_twist_02}
    componentAppliedLinearAccClamp=<5 5 5>
    componentLinearAccScale=<0.05 0.05 0.05>
    componentLinearVelScale=<0.05 0.05 0.05>
    gravityOverride {z=-980}
    physicsBodyDefinitions [{
      boundBone {boneName=dyn_lowerarm_r_sleeve_twist_02}
      boxExtents=<20 20 20>
      constraintSetup {angularLimitsMax=<1 2 2>, angularLimitsMin=<-1 -2 -2>, angularTarget=<1 0 0>}
      localJointOffset=<-1 0 0>
     }]
    relativeSpaceBone {boneName=root}
   }
   n17 = Anim Dynamics - Bone: dyn_lowerarm_r_sleeve_twist_03_end!(SimulationSpace:AnimPhysSimSpaceType Enum, ComponentPose=@n16, Alpha=0.8, ExternalForce="0,0,0") -> (Pose) {
    angularDampingOverride=1
    angularSpringConstant=1000
    bAngularSpring=true
    bLinearSpring=true
    bOverrideLinearDamping=true
    bUseGravityOverride=true
    bUsePlanarLimit=false
    gravityScale=0.35
    linearDampingOverride=1
    linearSpringConstant=100
    simulationSpace=RootRelative
    alphaBoolBlend {blendOption=Linear}
    boundBone {boneName=dyn_lowerarm_r_sleeve_twist_03_end}
    componentAppliedLinearAccClamp=<5 5 5>
    componentLinearAccScale=<0.05 0.05 0.05>
    componentLinearVelScale=<0.05 0.05 0.05>
    gravityOverride {z=-980}
    physicsBodyDefinitions [{
      boundBone {boneName=dyn_lowerarm_r_sleeve_twist_03_end}
      boxExtents=<15 15 15>
      constraintSetup {angularLimitsMax=<10 10 10>, angularLimitsMin=<-10 -10 -10>, angularTarget=<1 0 0>}
      localJointOffset=<-2 0 0>
     }]
    relativeSpaceBone {boneName=root}
   }
   n18 = Anim Dynamics - Bone: dyn_lowerarm_r_sleeve_twist_02_end!(SimulationSpace:AnimPhysSimSpaceType Enum, ComponentPose=@n17, Alpha=0.8, ExternalForce="0,0,0") -> (Pose) {
    angularDampingOverride=1
    angularSpringConstant=1000
    bAngularSpring=true
    bLinearSpring=true
    bOverrideLinearDamping=true
    bUseGravityOverride=true
    bUsePlanarLimit=false
    gravityScale=0.35
    linearDampingOverride=1
    linearSpringConstant=100
    simulationSpace=RootRelative
    alphaBoolBlend {blendOption=Linear}
    boundBone {boneName=dyn_lowerarm_r_sleeve_twist_02_end}
    componentAppliedLinearAccClamp=<5 5 5>
    componentLinearAccScale=<0.05 0.05 0.05>
    componentLinearVelScale=<0.05 0.05 0.05>
    gravityOverride {z=-980}
    physicsBodyDefinitions [{
      boundBone {boneName=dyn_lowerarm_r_sleeve_twist_02_end}
      boxExtents=<15 15 15>
      constraintSetup {angularLimitsMax=<10 10 10>, angularLimitsMin=<-10 -10 -10>, angularTarget=<1 0 0>}
      localJointOffset=<-2 0 0>
     }]
    relativeSpaceBone {boneName=root}
   }
   n19 = Anim Dynamics - Bone: dyn_lowerarm_l_sleeve_twist_01_MSTR!(SimulationSpace:AnimPhysSimSpaceType Enum, ComponentPose=@n18, Alpha=1, ExternalForce="0,0,0") -> (Pose) {
    bLinearSpring=true
    bOverrideLinearDamping=true
    linearDampingOverride=0.9999
    simulationSpace=RootRelative
    alphaBoolBlend {blendOption=Linear}
    boundBone {boneName=dyn_lowerarm_l_sleeve_twist_01_MSTR}
    physicsBodyDefinitions [{
      boundBone {boneName=dyn_lowerarm_l_sleeve_twist_01_MSTR}
      boxExtents=<300 300 300>
      constraintSetup {angularLimitsMax=<0 0 0>, angularLimitsMin=<0 0 0>, angularTarget=<0 0 0>, linearAxesMin=<-2 0 0>}
      localJointOffset=<1 0 0>
     }]
   }
   n20 = Anim Dynamics - Bone: dyn_lowerarm_l_sleeve_twist_02_MSTR!(SimulationSpace:AnimPhysSimSpaceType Enum, ComponentPose=@n19, Alpha=1, ExternalForce="0,0,0") -> (Pose) {
    bLinearSpring=true
    bOverrideLinearDamping=true
    linearDampingOverride=0.9999
    simulationSpace=RootRelative
    alphaBoolBlend {blendOption=Linear}
    boundBone {boneName=dyn_lowerarm_l_sleeve_twist_02_MSTR}
    physicsBodyDefinitions [{
      boundBone {boneName=dyn_lowerarm_l_sleeve_twist_02_MSTR}
      boxExtents=<300 300 300>
      constraintSetup {angularLimitsMax=<0 0 0>, angularLimitsMin=<0 0 0>, angularTarget=<0 0 0>, linearAxesMin=<-2 0 0>}
      localJointOffset=<1 0 0>
     }]
   }
   n21 = Anim Dynamics - Bone: dyn_lowerarm_l_sleeve_twist_03_MSTR!(SimulationSpace:AnimPhysSimSpaceType Enum, ComponentPose=@n20, Alpha=1, ExternalForce="0,0,0") -> (Pose) {
    bLinearSpring=true
    bOverrideLinearDamping=true
    linearDampingOverride=0.9999
    simulationSpace=RootRelative
    alphaBoolBlend {blendOption=Linear}
    boundBone {boneName=dyn_lowerarm_l_sleeve_twist_03_MSTR}
    physicsBodyDefinitions [{
      boundBone {boneName=dyn_lowerarm_l_sleeve_twist_03_MSTR}
      boxExtents=<300 300 300>
      constraintSetup {angularLimitsMax=<0 0 0>, angularLimitsMin=<0 0 0>, angularTarget=<0 0 0>, linearAxesMin=<-2 0 0>}
      localJointOffset=<1 0 0>
     }]
   }
   n22 = Anim Dynamics - Bone: dyn_lowerarm_r_sleeve_twist_01_MSTR!(SimulationSpace:AnimPhysSimSpaceType Enum, ComponentPose=@n21, Alpha=1, ExternalForce="0,0,0") -> (Pose) {
    bLinearSpring=true
    bOverrideLinearDamping=true
    linearDampingOverride=0.9999
    simulationSpace=RootRelative
    alphaBoolBlend {blendOption=Linear}
    boundBone {boneName=dyn_lowerarm_r_sleeve_twist_01_MSTR}
    physicsBodyDefinitions [{
      boundBone {boneName=dyn_lowerarm_r_sleeve_twist_01_MSTR}
      boxExtents=<300 300 300>
      constraintSetup {angularLimitsMax=<0 0 0>, angularLimitsMin=<0 0 0>, angularTarget=<0 0 0>, linearAxesMin=<-2 0 0>}
      localJointOffset=<-1 0 0>
     }]
   }
   n23 = Anim Dynamics - Bone: dyn_lowerarm_r_sleeve_twist_02_MSTR!(SimulationSpace:AnimPhysSimSpaceType Enum, ComponentPose=@n22, Alpha=1, ExternalForce="0,0,0") -> (Pose) {
    bLinearSpring=true
    bOverrideLinearDamping=true
    linearDampingOverride=0.9999
    simulationSpace=RootRelative
    alphaBoolBlend {blendOption=Linear}
    boundBone {boneName=dyn_lowerarm_r_sleeve_twist_02_MSTR}
    physicsBodyDefinitions [{
      boundBone {boneName=dyn_lowerarm_r_sleeve_twist_02_MSTR}
      boxExtents=<300 300 300>
      constraintSetup {angularLimitsMax=<0 0 0>, angularLimitsMin=<0 0 0>, angularTarget=<0 0 0>, linearAxesMin=<-2 0 0>}
      localJointOffset=<-1 0 0>
     }]
   }
   Anim Dynamics - Bone: dyn_lowerarm_r_sleeve_twist_03_MSTR!(SimulationSpace:AnimPhysSimSpaceType Enum, ComponentPose=@n23, Alpha=1, ExternalForce="0,0,0") -> (Pose) {
    bLinearSpring=true
    bOverrideLinearDamping=true
    linearDampingOverride=0.9999
    simulationSpace=RootRelative
    alphaBoolBlend {blendOption=Linear}
    boundBone {boneName=dyn_lowerarm_r_sleeve_twist_03_MSTR}
    physicsBodyDefinitions [{
      boundBone {boneName=dyn_lowerarm_r_sleeve_twist_03_MSTR}
      boxExtents=<300 300 300>
      constraintSetup {angularLimitsMax=<0 0 0>, angularLimitsMin=<0 0 0>, angularTarget=<0 0 0>, linearAxesMin=<-2 0 0>}
      localJointOffset=<-1 0 0>
     }]
   }
   n24 = "Constraint - dyn_underskirt_front_02 (RR)"!(ComponentPose, Alpha=1, ConstraintWeights_0=0.5, ConstraintWeights_1:float=0.5) -> (Pose) {
    alphaBoolBlend {blendOption=Linear}
    boneToModify {boneName=dyn_underskirt_front_02}
    constraintSetup [{transformType=Rotation, targetBone {boneName=dyn_underskirt_A_l_02}}, {transformType=Rotation, targetBone {boneName=dyn_underskirt_A_r_02}}]
    constraintWeights [1,1]
   }
   "Constraint - dyn_underskirt_front_03 (RR)"!(ComponentPose=@n24, Alpha=1, ConstraintWeights_0=0.5, ConstraintWeights_1:float=0.5) -> (Pose) {
    alphaBoolBlend {blendOption=Linear}
    boneToModify {boneName=dyn_underskirt_front_03}
    constraintSetup [{transformType=Rotation, targetBone {boneName=dyn_underskirt_A_l_03}}, {transformType=Rotation, targetBone {boneName=dyn_underskirt_A_r_03}}]
    constraintWeights [1,1]
   }
   n25 = "Constraint - dyn_underskirt_front_01 (PP)"!(ComponentPose, Alpha=1, ConstraintWeights_0=0.5, ConstraintWeights_1:float=0.5) -> (Pose) {
    alphaBoolBlend {blendOption=Linear}
    boneToModify {boneName=dyn_underskirt_front_01}
    constraintSetup [{targetBone {boneName=dyn_underskirt_A_l_01}}, {targetBone {boneName=dyn_underskirt_A_r_01}}]
    constraintWeights [1,1]
   }
   n26 = "Constraint - dyn_underskirt_front_02 (PP)"!(ComponentPose=@n25, Alpha=1, ConstraintWeights_0=0.5, ConstraintWeights_1:float=0.5) -> (Pose) {
    alphaBoolBlend {blendOption=Linear}
    boneToModify {boneName=dyn_underskirt_front_02}
    constraintSetup [{targetBone {boneName=dyn_underskirt_A_l_02}}, {targetBone {boneName=dyn_underskirt_A_r_02}}]
    constraintWeights [1,1]
   }
   "Constraint - dyn_underskirt_front_03 (PP)"!(ComponentPose=@n26, Alpha=1, ConstraintWeights_0=0.5, ConstraintWeights_1:float=0.5) -> (Pose) {
    alphaBoolBlend {blendOption=Linear}
    boneToModify {boneName=dyn_underskirt_front_03}
    constraintSetup [{targetBone {boneName=dyn_underskirt_A_l_03}}, {targetBone {boneName=dyn_underskirt_A_r_03}}]
    constraintWeights [1,1]
   }
   "Constraint - dyn_underskirt_front_MSTR (RR)"!(ComponentPose, Alpha=1, ConstraintWeights_0=0.5, ConstraintWeights_1:float=0.5) -> (Pose) {
    alphaBoolBlend {blendOption=Linear}
    boneToModify {boneName=dyn_underskirt_front_MSTR}
    constraintSetup [
     {transformType=Rotation, targetBone {boneName=dyn_underskirt_A_l_MSTR}}
     {transformType=Rotation, targetBone {boneName=dyn_underskirt_A_r_MSTR}}
    ]
    constraintWeights [1,1]
   }
   "Constraint - dyn_underskirt_front_MSTR (PP)"!(ComponentPose, Alpha=1, ConstraintWeights_0=0.5, ConstraintWeights_1:float=0.5) -> (Pose) {
    alphaBoolBlend {blendOption=Linear}
    boneToModify {boneName=dyn_underskirt_front_MSTR}
    constraintSetup [{targetBone {boneName=dyn_underskirt_A_r_MSTR}}, {targetBone {boneName=dyn_underskirt_A_l_MSTR}}]
    constraintWeights [1,1]
   }
   "Transform (Modify) Bone - Bone: dyn_sash_r_MSTR"(Alpha=Get thigh_r_up() -> (thigh_r_up:double), Translation="0,0,0", Rotation="0,-10,-20", Scale="1,1,1") {
    rotationMode=BMM_Additive
    rotationSpace=BCS_BoneSpace
    translationMode=BMM_Additive
    translationSpace=BCS_BoneSpace
    alphaBoolBlend {blendOption=Linear}
    boneToModify {boneName=dyn_sash_r_MSTR}
    rotation {roll=-20, yaw=-10}
   }
   n27 = "Transform (Modify) Bone - Bone: dyn_underskirt_A_l_MSTR"(Alpha=1, Translation="0,0,0", Rotation="0,0,0", Scale="1,1,1") {
    rotationMode=BMM_Additive
    rotationSpace=BCS_BoneSpace
    translationMode=BMM_Additive
    translationSpace=BCS_BoneSpace
    alphaBoolBlend {blendOption=Linear}
    boneToModify {boneName=dyn_underskirt_A_l_MSTR}
   }
   "Transform (Modify) Bone - Bone: dyn_underskirt_A_r_MSTR"(ComponentPose=@n27, Alpha=1, Translation="0,0,0", Rotation="0,0,0", Scale="1,1,1") {
    rotationMode=BMM_Additive
    rotationSpace=BCS_BoneSpace
    translationMode=BMM_Additive
    translationSpace=BCS_BoneSpace
    alphaBoolBlend {blendOption=Linear}
    boneToModify {boneName=dyn_underskirt_A_r_MSTR}
   }
   "Transform (Modify) Bone - Bone: dyn_underskirt_front_MSTR"(Alpha=1, Translation="0,0,0", Rotation="2,0,0", Scale="1,1,1") {
    rotationMode=BMM_Additive
    rotationSpace=BCS_BoneSpace
    translationMode=BMM_Additive
    translationSpace=BCS_BoneSpace
    alphaBoolBlend {blendOption=Linear}
    boneToModify {boneName=dyn_underskirt_front_MSTR}
    rotation {pitch=2}
   }
   "Transform (Modify) Bone - Bone: dyn_underskirt_front_MSTR"(Alpha=1, Translation="0,0,2", Rotation="4,0,0", Scale="1,1,1") {
    rotationMode=BMM_Additive
    rotationSpace=BCS_BoneSpace
    translationMode=BMM_Additive
    translationSpace=BCS_BoneSpace
    alphaBoolBlend {blendOption=Linear}
    boneToModify {boneName=dyn_underskirt_front_MSTR}
    rotation {pitch=4}
    translation {z=2}
   }
   n28 = "Transform (Modify) Bone - Bone: dyn_underskirt_A_l_MSTR"(Alpha=1, Translation="0,0,2", Rotation="4,0,0", Scale="1,1,1") {
    rotationMode=BMM_Additive
    rotationSpace=BCS_BoneSpace
    translationMode=BMM_Additive
    translationSpace=BCS_BoneSpace
    alphaBoolBlend {blendOption=Linear}
    boneToModify {boneName=dyn_underskirt_A_l_MSTR}
    rotation {pitch=4}
    translation {z=2}
   }
   "Transform (Modify) Bone - Bone: dyn_underskirt_A_r_MSTR"(ComponentPose=@n28, Alpha=1, Translation="0,0,-4", Rotation="4,0,0", Scale="1,1,1") {
    rotationMode=BMM_Additive
    rotationSpace=BCS_BoneSpace
    translationMode=BMM_Additive
    translationSpace=BCS_BoneSpace
    alphaBoolBlend {blendOption=Linear}
    boneToModify {boneName=dyn_underskirt_A_r_MSTR}
    rotation {pitch=4}
    translation {z=-4}
   }
   n29 = Control Rig!(Source=Copy Pose From Mesh() {bUseAttachedParent=true}, Alpha, isKneeSliding:bool=Get bIsKneeSliding(), isFrontEnd:bool=Get bIsInFrontEnd()) -> (Pose) {
    alphaBoolBlend {blendOption=Linear}
    controlRigAssetReference {blueprintRigClass=&/M_MED_GeneAglet_Fly_CtrlRig.M_MED_GeneAglet_Fly_CtrlRig_C}
   }
   n30 = RigidBody!(
    ComponentPose=Local To Component(LocalPose=@n29)
    Alpha
    OverrideWorldGravity:Vector="0,0,-980"
    ExternalForce
    SimSpaceSettings:Sim Space Settings Structure=(WorldAlpha=1,VelocityScaleZ=1,DampingAlpha=0,MaxLinearVelocity=700,MaxAngularVelocity=8,MaxLinearAcceleration=3000,MaxAngularAcceleration=50,ExternalLinearDragV=(X=0.5,Y=0.5,Z=0.25),ExternalLinearVelocity=(X=0,Y=0,Z=0),ExternalAngularVelocity=(X=0,Y=0,Z=0))
   ) -> (Pose) {
    bOverrideWorldGravity=true
    overridePhysicsAsset=&/M_MED_GeneAglet_Fly_Physics
    simulationSpace=BaseBoneSpace
    alphaBoolBlend {blendOption=Linear}
    baseBoneRef {boneName=pelvis}
    overrideWorldGravity {z=-980}
    simSpaceSettings {
     dampingAlpha=0
     maxAngularAcceleration=50
     maxAngularVelocity=8
     maxLinearAcceleration=3000
     maxLinearVelocity=700
     worldAlpha=1
     externalLinearDragV=<0.5 0.5 0.25>
    }
   }
   n31 = RigidBody!(ComponentPose=@n30, Alpha, OverrideWorldGravity="0,0,-980", ExternalForce) -> (Pose) {
    bOverrideWorldGravity=true
    overridePhysicsAsset=&/M_MED_GeneAglet_Fly_SleeveL_Physics
    simulationSpace=BaseBoneSpace
    alphaBoolBlend {blendOption=Linear}
    baseBoneRef {boneName=lowerarm_l}
    componentAppliedLinearAccClamp=<0 0 0>
    overrideWorldGravity {z=-980}
    simSpaceSettings {
     dampingAlpha=0
     maxAngularAcceleration=30
     maxAngularVelocity=5
     maxLinearAcceleration=1000
     maxLinearVelocity=100
     velocityScaleZ=0.2
     worldAlpha=0.8
     externalLinearDragV=<0.5 0.5 0.5>
    }
   }
   n32 = RigidBody(ComponentPose=@n31, OverrideWorldGravity="0,0,-980") {
    bOverrideWorldGravity=true
    overridePhysicsAsset=&/M_MED_GeneAglet_Fly_SleeveR_Physics
    simulationSpace=BaseBoneSpace
    alphaBoolBlend {blendOption=Linear}
    baseBoneRef {boneName=lowerarm_r}
    componentAppliedLinearAccClamp=<0 0 0>
    overrideWorldGravity {z=-980}
    simSpaceSettings {
     dampingAlpha=0
     maxAngularAcceleration=30
     maxAngularVelocity=5
     maxLinearAcceleration=1000
     maxLinearVelocity=100
     velocityScaleZ=0.2
     worldAlpha=0.8
     externalLinearDragV=<0.5 0.5 0.5>
    }
   }
   n33 = Anim Dynamics - Bone: dyn_lightsaber!(SimulationSpace:AnimPhysSimSpaceType Enum, ComponentPose=@n32, Alpha=1, GravityOverride="0,0,-980", ExternalForce="0,0,0") -> (Pose) {
    angularDampingOverride=0.9999
    angularSpringConstant=100
    bAngularSpring=true
    bOverrideAngularDamping=true
    bUseGravityOverride=true
    simulationSpace=RootRelative
    alphaBoolBlend {blendOption=Linear}
    boundBone {boneName=dyn_lightsaber}
    componentLinearAccScale=<0.4 0.4 0.4>
    componentLinearVelScale=<0.5 0.5 0.5>
    gravityOverride {z=-980}
    physicsBodyDefinitions [{
      boundBone {boneName=dyn_lightsaber}
      boxExtents=<16 16 16>
      constraintSetup {angularLimitsMax=<0 10 6>, angularLimitsMin=<0 -10 -6>, angularTarget=<1 0 0>}
      localJointOffset=<2.5 0 0>
     }]
   }
   n34 = "Transform (Modify) Bone - Bone: dyn_underskirt_front_MSTR"(ComponentPose=@n33, Alpha=1, Translation="0,0,0", Rotation="15,0,0", Scale="1,1,1") {
    rotationMode=BMM_Additive
    rotationSpace=BCS_BoneSpace
    translationMode=BMM_Additive
    translationSpace=BCS_BoneSpace
    alphaBoolBlend {blendOption=Linear}
    boneToModify {boneName=dyn_underskirt_front_MSTR}
    rotation {pitch=15}
   }
   n35 = "Transform (Modify) Bone - Bone: dyn_underskirt_A_r_MSTR"(ComponentPose=@n34, Alpha=1, Translation="0,0,0", Rotation="0,0,0", Scale="1,1,1") {
    rotationMode=BMM_Additive
    rotationSpace=BCS_BoneSpace
    translationMode=BMM_Additive
    translationSpace=BCS_BoneSpace
    alphaBoolBlend {blendOption=Linear}
    boneToModify {boneName=dyn_underskirt_A_r_MSTR}
   }
   n36 = "Transform (Modify) Bone - Bone: dyn_underskirt_A_l_MSTR"(ComponentPose=@n35, Alpha=1, Translation="0,0,0", Rotation="13.662,26.439,0.000", Scale="1,1,1") {
    rotationMode=BMM_Additive
    rotationSpace=BCS_BoneSpace
    translationMode=BMM_Additive
    translationSpace=BCS_BoneSpace
    alphaBoolBlend {blendOption=Linear}
    boneToModify {boneName=dyn_underskirt_A_l_MSTR}
    rotation=<rot 13.662 26.439 0>
   }
   n37 = "Transform (Modify) Bone - Bone: dyn_underskirt_A_r_MSTR"(ComponentPose=@n36, Alpha=1, Translation="0,0,-1", Rotation="2,-7,-10", Scale="1,1,1") {
    rotationMode=BMM_Additive
    rotationSpace=BCS_BoneSpace
    translationMode=BMM_Additive
    translationSpace=BCS_BoneSpace
    alphaBoolBlend {blendOption=Linear}
    boneToModify {boneName=dyn_underskirt_A_r_MSTR}
    rotation=<rot 2 -7 -10>
    translation {z=-1}
   }
   n38 = "Constraint - dyn_sash_l_01 (P)"(ComponentPose=@n37, Alpha=1, ConstraintWeights_0=1) {
    alphaBoolBlend {blendOption=Linear}
    boneToModify {boneName=dyn_sash_l_01}
    constraintSetup [{targetBone {boneName=dyn_underskirt_A_l_01}}]
    constraintWeights [1]
   }
   n39 = "Constraint - dyn_sash_l_02 (P)"(ComponentPose=@n38, Alpha=1, ConstraintWeights_0=1) {
    alphaBoolBlend {blendOption=Linear}
    boneToModify {boneName=dyn_sash_l_02}
    constraintSetup [{targetBone {boneName=dyn_underskirt_A_l_02}}]
    constraintWeights [1]
   }
   n40 = "Constraint - dyn_sash_l_03 (P)"(ComponentPose=@n39, Alpha=1, ConstraintWeights_0=1) {
    alphaBoolBlend {blendOption=Linear}
    boneToModify {boneName=dyn_sash_l_03}
    constraintSetup [{targetBone {boneName=dyn_underskirt_A_l_03}}]
    constraintWeights [1]
   }
   n41 = "Constraint - dyn_sash_r_01 (P)"(ComponentPose=@n40, Alpha=1, ConstraintWeights_0=1) {
    alphaBoolBlend {blendOption=Linear}
    boneToModify {boneName=dyn_sash_r_01}
    constraintSetup [{targetBone {boneName=dyn_underskirt_A_r_01}}]
    constraintWeights [1]
   }
   n42 = "Constraint - dyn_sash_r_02 (P)"(ComponentPose=@n41, Alpha=1, ConstraintWeights_0=1) {
    alphaBoolBlend {blendOption=Linear}
    boneToModify {boneName=dyn_sash_r_02}
    constraintSetup [{targetBone {boneName=dyn_underskirt_A_r_02}}]
    constraintWeights [1]
   }
   n43 = "Constraint - dyn_sash_r_03 (P)"(ComponentPose=@n42, Alpha=1, ConstraintWeights_0=1) {
    alphaBoolBlend {blendOption=Linear}
    boneToModify {boneName=dyn_sash_r_03}
    constraintSetup [{targetBone {boneName=dyn_underskirt_A_r_03}}]
    constraintWeights [1]
   }
   n44 = "Transform (Modify) Bone - Bone: dyn_sash_r_MSTR"(ComponentPose=@n43, Alpha=1, Translation="0,0,-1", Rotation="4,0,-10", Scale="1,1,1") {
    rotationMode=BMM_Additive
    rotationSpace=BCS_BoneSpace
    translationMode=BMM_Additive
    translationSpace=BCS_BoneSpace
    alphaBoolBlend {blendOption=Linear}
    boneToModify {boneName=dyn_sash_r_MSTR}
    rotation {pitch=4, roll=-10}
    translation {z=-1}
   }
   n45 = "Transform (Modify) Bone - Bone: dyn_sash_l_MSTR"(ComponentPose=@n44, Alpha=1, Translation="0,0,1", Rotation="4,0,10", Scale="1,1,1") {
    rotationMode=BMM_Additive
    rotationSpace=BCS_BoneSpace
    translationMode=BMM_Additive
    translationSpace=BCS_BoneSpace
    alphaBoolBlend {blendOption=Linear}
    boneToModify {boneName=dyn_sash_l_MSTR}
    rotation {pitch=4, roll=10}
    translation {z=1}
   }
   n46 = "Transform (Modify) Bone - Bone: dyn_sash_r_MSTR"(ComponentPose=@n45, Alpha=1, Translation="0,0,0", Rotation="0,-4,-10", Scale="1,1,1") {
    rotationMode=BMM_Additive
    rotationSpace=BCS_BoneSpace
    translationMode=BMM_Additive
    translationSpace=BCS_BoneSpace
    alphaBoolBlend {blendOption=Linear}
    boneToModify {boneName=dyn_sash_r_MSTR}
    rotation {roll=-10, yaw=-4}
   }
   n47 = Anim Dynamics - Bone: dyn_sash_l_01!(SimulationSpace:AnimPhysSimSpaceType Enum, ComponentPose=@n46, Alpha=1, GravityOverride:Vector="0,0,-980", ExternalForce="0,0,0") -> (Pose) {
    angularDampingOverride=0.9999
    angularSpringConstant=200
    bAngularSpring=true
    bOverrideAngularDamping=true
    bUseGravityOverride=true
    simulationSpace=RootRelative
    alphaBoolBlend {blendOption=Linear}
    boundBone {boneName=dyn_sash_l_01}
    componentLinearAccScale=<0.4 0.4 0.4>
    componentLinearVelScale=<0.5 0.5 0.5>
    gravityOverride {z=-980}
    physicsBodyDefinitions [{
      boundBone {boneName=dyn_sash_l_01}
      boxExtents=<10 10 10>
      constraintSetup {angularLimitsMax=<0 0 0>, angularLimitsMin=<0 -2 -4>, angularTarget=<1 0 0>}
      localJointOffset=<1 0 0>
     }]
   }
   n48 = Anim Dynamics - Bone: dyn_sash_l_02!(SimulationSpace:AnimPhysSimSpaceType Enum=Component, ComponentPose=@n47, Alpha=1, GravityOverride:Vector="0,0,-980", ExternalForce="0,0,0") -> (Pose) {
    angularDampingOverride=0.999
    angularSpringConstant=200
    bAngularSpring=true
    bOverrideAngularDamping=true
    bUseGravityOverride=true
    simulationSpace=RootRelative
    alphaBoolBlend {blendOption=Linear}
    boundBone {boneName=dyn_sash_l_02}
    componentLinearAccScale=<0.4 0.4 0.4>
    componentLinearVelScale=<0.5 0.5 0.5>
    gravityOverride {z=-980}
    physicsBodyDefinitions [{
      boundBone {boneName=dyn_sash_l_02}
      boxExtents=<10 10 10>
      constraintSetup {angularLimitsMax=<0 0 0>, angularLimitsMin=<0 -4 -8>, angularTarget=<1 0 0>}
      localJointOffset=<1 0 0>
     }]
   }
   n49 = Anim Dynamics - Bone: dyn_sash_l_03!(SimulationSpace:AnimPhysSimSpaceType Enum, ComponentPose=@n48, Alpha=1, GravityOverride:Vector="0,0,-980", ExternalForce="0,0,0") -> (Pose) {
    angularDampingOverride=0.999
    angularSpringConstant=100
    bAngularSpring=true
    bOverrideAngularDamping=true
    bUseGravityOverride=true
    simulationSpace=RootRelative
    alphaBoolBlend {blendOption=Linear}
    boundBone {boneName=dyn_sash_l_03}
    componentLinearAccScale=<0.4 0.4 0.4>
    componentLinearVelScale=<0.5 0.5 0.5>
    gravityOverride {z=-980}
    physicsBodyDefinitions [{
      boundBone {boneName=dyn_sash_l_03}
      boxExtents=<10 10 10>
      constraintSetup {angularLimitsMax=<0 0 0>, angularLimitsMin=<0 -4 -10>, angularTarget=<1 0 0>}
      localJointOffset=<1 0 0>
     }]
   }
   n50 = Anim Dynamics - Bone: dyn_sash_r_01!(SimulationSpace:AnimPhysSimSpaceType Enum, ComponentPose=@n49, Alpha=1, GravityOverride:Vector="0,0,-980", ExternalForce="0,0,0") -> (Pose) {
    angularDampingOverride=0.9999
    angularSpringConstant=200
    bAngularSpring=true
    bOverrideAngularDamping=true
    bUseGravityOverride=true
    simulationSpace=RootRelative
    alphaBoolBlend {blendOption=Linear}
    boundBone {boneName=dyn_sash_r_01}
    componentLinearAccScale=<0.4 0.4 0.4>
    componentLinearVelScale=<0.5 0.5 0.5>
    gravityOverride {z=-980}
    physicsBodyDefinitions [{
      boundBone {boneName=dyn_sash_r_01}
      boxExtents=<10 10 10>
      constraintSetup {angularLimitsMax=<0 0 0>, angularLimitsMin=<0 -2 -2>, angularTarget=<1 0 0>}
      localJointOffset=<-1 0 0>
     }]
   }
   n51 = Anim Dynamics - Bone: dyn_sash_r_02!(SimulationSpace:AnimPhysSimSpaceType Enum, ComponentPose=@n50, Alpha=1, GravityOverride:Vector="0,0,-980", ExternalForce="0,0,0") -> (Pose) {
    angularDampingOverride=0.999
    angularSpringConstant=200
    bAngularSpring=true
    bOverrideAngularDamping=true
    bUseGravityOverride=true
    simulationSpace=RootRelative
    alphaBoolBlend {blendOption=Linear}
    boundBone {boneName=dyn_sash_r_02}
    componentLinearAccScale=<0.4 0.4 0.4>
    componentLinearVelScale=<0.5 0.5 0.5>
    gravityOverride {z=-980}
    physicsBodyDefinitions [{
      boundBone {boneName=dyn_sash_r_02}
      boxExtents=<10 10 10>
      constraintSetup {angularLimitsMax=<0 0 -8>, angularLimitsMin=<0 -4 -8>, angularTarget=<1 0 0>}
      localJointOffset=<-1 0 0>
     }]
   }
   n52 = Anim Dynamics - Bone: dyn_sash_r_03!(SimulationSpace:AnimPhysSimSpaceType Enum, ComponentPose=@n51, Alpha=1, GravityOverride:Vector="0,0,-980", ExternalForce="0,0,0") -> (Pose) {
    angularDampingOverride=0.999
    angularSpringConstant=100
    bAngularSpring=true
    bOverrideAngularDamping=true
    bUseGravityOverride=true
    simulationSpace=RootRelative
    alphaBoolBlend {blendOption=Linear}
    boundBone {boneName=dyn_sash_r_03}
    componentLinearAccScale=<0.4 0.4 0.4>
    componentLinearVelScale=<0.5 0.5 0.5>
    gravityOverride {z=-980}
    physicsBodyDefinitions [{
      boundBone {boneName=dyn_sash_r_03}
      boxExtents=<10 10 10>
      constraintSetup {angularLimitsMax=<0 0 0>, angularLimitsMin=<0 -4 -10>, angularTarget=<1 0 0>}
      localJointOffset=<-1 0 0>
     }]
   }
   n53 = "Transform (Modify) Bone - Bone: dyn_sash_r_MSTR"!(ComponentPose=@n52, bAlphaBoolEnabled:bool=True, Translation="0,0,0", Rotation="-3,-0.249,2", Scale="1,1,1") -> (Pose) {
    alphaInputType=Bool
    rotationMode=BMM_Additive
    rotationSpace=BCS_BoneSpace
    alphaBoolBlend {blendOption=Linear}
    boneToModify {boneName=dyn_sash_r_MSTR}
    rotation=<rot -3 -0.249 2>
   }
   n54 = "Transform (Modify) Bone - Bone: dyn_sash_r_02"!(ComponentPose=@n53, bAlphaBoolEnabled:bool=True, Translation="0,0,0", Rotation="-2,-4.739,-2.097", Scale="1,1,1") -> (Pose) {
    alphaInputType=Bool
    rotationMode=BMM_Additive
    rotationSpace=BCS_BoneSpace
    alphaBoolBlend {blendOption=Linear}
    boneToModify {boneName=dyn_sash_r_02}
    rotation=<rot -2 -4.739 -2.097>
   }
   n55 = "Transform (Modify) Bone - Bone: dyn_sash_l_MSTR"!(ComponentPose=@n54, bAlphaBoolEnabled:bool=True, Translation="0,0,0", Rotation="-1.896,-0.600,-12.637", Scale="1,1,1") -> (Pose) {
    alphaInputType=Bool
    rotationMode=BMM_Additive
    rotationSpace=BCS_BoneSpace
    alphaBoolBlend {blendOption=Linear}
    boneToModify {boneName=dyn_sash_l_MSTR}
    rotation=<rot -1.8959 -0.6 -12.637>
   }
   n56 = "Transform (Modify) Bone - Bone: dyn_sash_l_02"!(ComponentPose=@n55, bAlphaBoolEnabled:bool=True, Translation="0,0,0", Rotation="-1.400,-2.834,0.001", Scale="1,1,1") -> (Pose) {
    alphaInputType=Bool
    rotationMode=BMM_Additive
    rotationSpace=BCS_BoneSpace
    alphaBoolBlend {blendOption=Linear}
    boneToModify {boneName=dyn_sash_l_02}
    rotation=<rot -1.4 -2.834 0.001>
   }
   n57 = "Transform (Modify) Bone - Bone: dyn_sash_l_03"!(ComponentPose=@n56, bAlphaBoolEnabled:bool=True, Translation="0,0,0", Rotation="-0.001,5.553,0.012", Scale="1,1,1") -> (Pose) {
    alphaInputType=Bool
    rotationMode=BMM_Additive
    rotationSpace=BCS_BoneSpace
    alphaBoolBlend {blendOption=Linear}
    boneToModify {boneName=dyn_sash_l_03}
    rotation=<rot -0.001 5.553 0.012>
   }
   n58 = "Transform (Modify) Bone - Bone: dyn_lightsaber_MSTR"!(ComponentPose=@n57, bAlphaBoolEnabled:bool=True, Translation="0,0,0", Rotation="-6,0,0", Scale="1,1,1") -> (Pose) {
    alphaInputType=Bool
    rotationMode=BMM_Additive
    rotationSpace=BCS_BoneSpace
    alphaBoolBlend {blendOption=Linear}
    boneToModify {boneName=dyn_lightsaber_MSTR}
    rotation {pitch=-6}
   }
   n59 = "Transform (Modify) Bone - Bone: dyn_lowerarm_sleeve_r_MSTR"!(ComponentPose=@n58, bAlphaBoolEnabled:bool=True, Translation="0,-2,0", Rotation="0,0,0", Scale="1,1,1") -> (Pose) {
    alphaInputType=Bool
    rotationMode=BMM_Additive
    rotationSpace=BCS_BoneSpace
    translationMode=BMM_Additive
    translationSpace=BCS_BoneSpace
    alphaBoolBlend {blendOption=Linear}
    boneToModify {boneName=dyn_lowerarm_sleeve_r_MSTR}
    translation {y=-2}
   }
   n60 = "Transform (Modify) Bone - Bone: dyn_sash_l_MSTR"(ComponentPose=@n59, bAlphaBoolEnabled=True, Translation="0,0,0", Rotation="-5,-0.043,0.037", Scale="1,1,1") {
    alphaInputType=Bool
    rotationMode=BMM_Additive
    rotationSpace=BCS_BoneSpace
    alphaBoolBlend {blendOption=Linear}
    boneToModify {boneName=dyn_sash_l_MSTR}
    rotation=<rot -5 -0.043 0.037>
   }
   n61 = "Transform (Modify) Bone - Bone: dyn_sash_r_MSTR"(ComponentPose=@n60, bAlphaBoolEnabled=True, Translation="0,0,0", Rotation="-4,0,0", Scale="1,1,1") {
    alphaInputType=Bool
    rotationMode=BMM_Additive
    rotationSpace=BCS_BoneSpace
    alphaBoolBlend {blendOption=Linear}
    boneToModify {boneName=dyn_sash_r_MSTR}
    rotation {pitch=-4}
   }
   n62 = "Transform (Modify) Bone - Bone: dyn_sash_l_02"(ComponentPose=@n61, bAlphaBoolEnabled=True, Translation="0,0,0", Rotation="-0.190,-7.060,0.064", Scale="1,1,1") {
    alphaInputType=Bool
    rotationMode=BMM_Additive
    rotationSpace=BCS_BoneSpace
    alphaBoolBlend {blendOption=Linear}
    boneToModify {boneName=dyn_sash_l_02}
    rotation=<rot -0.19 -7.06 0.064>
   }
   n63 = "Transform (Modify) Bone - Bone: dyn_sash_r_02"(ComponentPose=@n62, bAlphaBoolEnabled=True, Translation="0,0,0", Rotation="-0.125,10.077,-0.027", Scale="1,1,1") {
    alphaInputType=Bool
    rotationMode=BMM_Additive
    rotationSpace=BCS_BoneSpace
    alphaBoolBlend {blendOption=Linear}
    boneToModify {boneName=dyn_sash_r_02}
    rotation=<rot -0.1249 10.077 -0.027>
   }
   Output Pose(Result=Component To Local(ComponentPose=@n63)) {layerGroup=None}
   GetCurveValue(CurveName=thigh_l_up)
  ]}
 EventGraph {nodes [
   Event BlueprintUpdateAnimation()
   > n1 = Sequence!(execute) -> (then_0, then_1, then_2:Exec, then_3:Exec)
   > n2 = Gravity Override State Function DEPRECATED(
    CharacterPartAnimInstance=Self-Reference()
    GameTimeInSeconds=GetGameTimeInSeconds()
    GravityOverrideParams=Get Gravity Override Params()
    RBAN Params=Get RBAN Params()
   )
   > Set Gravity Override(Gravity Override=@n2)
   > Set RBAN Settings Out(RBAN Settings Out:RBANSetting Struct Structure=@n2.RBAN SettingsOut) -> (Output_Get:RBANSetting Struct Structure)
   n3 = Branch(execute=@n1.then_1, Condition=Get IsOnZipline() -> (IsOnZipline:bool))
   > Set AnimDynamicsAlpha(AnimDynamicsAlpha:double=0) -> (Output_Get:double)
   Set AnimDynamicsAlpha(execute=@n3.else, AnimDynamicsAlpha=1)
   Set thigh_r_up(execute=@n1.then_2, thigh_r_up:double=GetCurveValue(CurveName=thigh_r_up)) -> (Output_Get:double)
   > Set thigh_l_across(thigh_l_across:float=GetCurveValue(CurveName=thigh_l_across)) -> (Output_Get:float)
   n4 = Branch(execute=@n1.then_3, Condition=Get bIsInFrontEnd())
   > Set Simulation Space(Simulation Space:AnimPhysSimSpaceType Enum=World) -> (Output_Get:AnimPhysSimSpaceType Enum)
   > Set AnimDynamicsAlpha(AnimDynamicsAlpha=0.1)
   > Set SashAdjustAlpha(SashAdjustAlpha:double=0.65) -> (Output_Get:double)
   > Set InGame(InGame:bool=false) -> (Output_Get:bool)
   Set Simulation Space(execute=@n4.else)
   > Set AnimDynamicsAlpha(AnimDynamicsAlpha=1)
   > Set SashAdjustAlpha(SashAdjustAlpha=1)
   > Set InGame(InGame=true)
   n5 = Event BlueprintProcessAnimInputEvent()
   > n6 = Branch(Condition=IsNamed(self=@n5.AnimInputEvent, InEventName=aie.Player.Zipline.Entered))
   > Set IsOnZipline(IsOnZipline:bool=true) -> (Output_Get:bool)
   Branch(execute=@n6.else, Condition=IsNamed(self=@n5.AnimInputEvent, InEventName=aie.Player.Zipline.Exited))
   > Set IsOnZipline(IsOnZipline=false)
  ]}
]
variables [
 AnimDynamicsAlpha {type=number, default=1}
 Gravity Override {type=Vector}
 Gravity Override Params {
  type=GravityOverrideParamsStruct
  default {
   onGround_Moving {gravityOverride=<0 -200 -980>}
   skydive_Parachuting {jointName=pelvis, gravityOverride=<0 -800 0>}
   tactical_Sprint {gravityOverride=<0 -450 -100>}
  }
 }
 InGame {type=boolean}
 IsOnZipline {type=boolean}
 KneeSlide {type=boolean}
 RBAN Params {
  type=RBANParamsStruct
  default {
   dBNO {simSpaceSettings {
     dampingAlpha=0
     maxAngularAcceleration=30
     maxAngularVelocity=7.5
     maxLinearAcceleration=3000
     maxLinearVelocity=400
     velocityScaleZ=0.4
     externalLinearDragV=<0.5 0.5 0.25>
    }}
   dBNO_Carried {simSpaceSettings {
     dampingAlpha=0
     maxAngularAcceleration=30
     maxAngularVelocity=7.5
     maxLinearAcceleration=3000
     maxLinearVelocity=400
     velocityScaleZ=0.4
     externalLinearDragV=<0.5 0.5 0.25>
    }}
   emoteOrMelee {simSpaceSettings {
     dampingAlpha=0
     maxAngularAcceleration=30
     maxAngularVelocity=7.5
     maxLinearAcceleration=3000
     maxLinearVelocity=400
     velocityScaleZ=0.4
     externalLinearDragV=<0.5 0.5 0.25>
    }}
   front_End {simSpaceSettings {
     dampingAlpha=0
     maxAngularAcceleration=20
     maxAngularVelocity=4
     maxLinearAcceleration=3000
     maxLinearVelocity=400
     velocityScaleZ=0.4
     externalLinearDragV=<0.5 0.5 0.25>
    }}
   onGround_Crouch_Moving {simSpaceSettings {
     dampingAlpha=0
     maxAngularAcceleration=30
     maxAngularVelocity=7.5
     maxLinearAcceleration=3000
     maxLinearVelocity=400
     velocityScaleZ=0.4
     externalLinearDragV=<0.5 0.5 0.25>
    }}
   onGround_Falling {simSpaceSettings {
     dampingAlpha=0
     maxAngularAcceleration=30
     maxAngularVelocity=7.5
     maxLinearAcceleration=3000
     maxLinearVelocity=400
     velocityScaleZ=0.4
     externalLinearDragV=<0.5 0.5 0.25>
    }}
   onGround_Moving {
    componentLinearAccScale=<0 0 0>
    componentLinearVelScale=<0 0 0>
    simSpaceSettings {
     dampingAlpha=0
     maxAngularAcceleration=30
     maxAngularVelocity=7.5
     maxLinearAcceleration=1000
     maxLinearVelocity=200
     velocityScaleZ=1
     externalLinearDragV=<0.7 0.7 0.5>
    }
   }
   onGround_Standing {
    componentLinearAccScale=<0 0 0>
    componentLinearVelScale=<0 0 0>
    simSpaceSettings {
     dampingAlpha=0
     maxAngularAcceleration=30
     maxAngularVelocity=7.5
     maxLinearAcceleration=3000
     maxLinearVelocity=400
     velocityScaleZ=0.4
     externalLinearDragV=<0.5 0.5 0.25>
    }
   }
   rail_Grinding {simSpaceSettings {
     dampingAlpha=0
     maxAngularAcceleration=30
     maxAngularVelocity=7.5
     maxLinearAcceleration=3000
     maxLinearVelocity=400
     velocityScaleZ=0.4
     worldAlpha=1
     externalLinearDragV=<0.5 0.5 0.25>
    }}
   riding_Moving {simSpaceSettings {
     dampingAlpha=0
     maxAngularAcceleration=30
     maxAngularVelocity=7.5
     maxLinearAcceleration=3000
     maxLinearVelocity=400
     velocityScaleZ=0.4
     externalLinearDragV=<0.5 0.5 0.25>
    }}
   riding_Moving_Targeting {simSpaceSettings {
     dampingAlpha=0
     maxAngularAcceleration=30
     maxAngularVelocity=7.5
     maxLinearAcceleration=3000
     maxLinearVelocity=400
     velocityScaleZ=0.4
     externalLinearDragV=<0.5 0.5 0.25>
    }}
   riding_Standing {simSpaceSettings {
     dampingAlpha=0
     maxAngularAcceleration=30
     maxAngularVelocity=7.5
     maxLinearAcceleration=3000
     maxLinearVelocity=400
     velocityScaleZ=0.4
     externalLinearDragV=<0.5 0.5 0.25>
    }}
   skydive_DivingDown {
    componentLinearAccScale=<0 0 0>
    componentLinearVelScale=<0 0 0>
    simSpaceSettings {
     dampingAlpha=0
     maxAngularAcceleration=20
     maxAngularVelocity=5
     maxLinearAcceleration=2000
     maxLinearVelocity=400
     velocityScaleZ=1
     externalLinearDragV=<0.5 0.5 0.25>
    }
   }
   skydive_FloatingDown {
    componentLinearAccScale=<0 0 0>
    componentLinearVelScale=<0 0 0>
    simSpaceSettings {
     dampingAlpha=0
     maxAngularAcceleration=20
     maxAngularVelocity=5
     maxLinearAcceleration=2000
     maxLinearVelocity=400
     velocityScaleZ=1
     externalLinearDragV=<0.5 0.5 0.25>
    }
   }
   skydive_GoingUp {simSpaceSettings {
     dampingAlpha=0
     maxAngularAcceleration=20
     maxAngularVelocity=5
     maxLinearAcceleration=2000
     maxLinearVelocity=400
     velocityScaleZ=1
     externalLinearDragV=<0.5 0.5 0.25>
    }}
   skydive_Parachuting {
    componentLinearAccScale=<0 0 0>
    componentLinearVelScale=<0 0 0>
    simSpaceSettings {
     dampingAlpha=0
     maxAngularAcceleration=20
     maxAngularVelocity=5
     maxLinearAcceleration=1000
     maxLinearVelocity=200
     velocityScaleZ=1
     externalLinearDragV=<0.5 0.5 0.25>
    }
   }
   skydive_Surfing {simSpaceSettings {
     dampingAlpha=0
     maxAngularAcceleration=30
     maxAngularVelocity=7.5
     maxLinearAcceleration=3000
     maxLinearVelocity=400
     velocityScaleZ=0.4
     worldAlpha=0.7
     externalLinearDragV=<0.5 0.5 0.25>
    }}
   swimming {simSpaceSettings {
     dampingAlpha=0
     maxAngularAcceleration=30
     maxAngularVelocity=7.5
     maxLinearAcceleration=3000
     maxLinearVelocity=400
     velocityScaleZ=0.4
     externalLinearDragV=<0.5 0.5 0.25>
    }}
   tactical_Sprint {
    componentLinearAccScale=<0 0 0>
    componentLinearVelScale=<0 0 0>
    simSpaceSettings {
     dampingAlpha=0
     maxAngularAcceleration=30
     maxAngularVelocity=7.5
     maxLinearAcceleration=3000
     maxLinearVelocity=400
     velocityScaleZ=0.2
     externalLinearDragV=<0.7 0.7 0.5>
    }
   }
   ziplining {simSpaceSettings {
     dampingAlpha=0
     maxAngularAcceleration=30
     maxAngularVelocity=7.5
     maxLinearAcceleration=3000
     maxLinearVelocity=400
     velocityScaleZ=0.4
     externalLinearDragV=<0.5 0.5 0.25>
    }}
  }
 }
 RBAN Settings Out {
  type=RBANSettingStruct
  default {
   componentLinearAccScale=<0 0 0>
   componentLinearVelScale=<0 0 0>
   simSpaceSettings {
    dampingAlpha=1
    maxAngularAcceleration=1000
    maxAngularVelocity=3
    maxLinearAcceleration=300
    maxLinearVelocity=700
    velocityScaleZ=0
    worldAlpha=1
    externalAngularVelocity=<0 0 0>
    externalLinearDragV=<0 0.5 0>
    externalLinearVelocity=<0 0 0>
   }
  }
 }
 SashAdjustAlpha {type=number}
 Simulation Space {type=AnimPhysSimSpaceType, default=RootRelative}
 thigh_l_across {type=number}
 thigh_l_up {type=number, default=1}
 thigh_r_up {type=number, default=1}
]
